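# Reticle extension widget for dual-magazine weapons that shows two radial ammo gauges and highlights the active magazine.
# On construct it binds to weapon ammo and DualMag reload-speed events, then snaps to the current magazine after a short settle timer.
# Swaps play SwapMagazine forward or reversed and redistribute ammo through dynamic material parameters, which also carry segments and ADS state.
# Timers hold the gauges visible during reloads, then fade them out.
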WidgetBlueprint /Game/Athena/HUD/reticles/ReticleExtensionWidgets/WBP_DualMag_Reticle
animations [
 FadeOut {
  binding_count=1
  length_seconds=0.4969
  object=&self:FadeOut
  sequence {bindings [Overlay_Magazines {
     id=#1
     possessed_class=Overlay
     tracks [{
       type=MovieSceneFloatTrack
       display_name=Render Opacity
       sections [{type=MovieSceneFloatSection, channels [None {type=MovieSceneScriptingFloatChannel, keys [[0,1],[9.937,0]]}]}]
      }]
    }]}
 }
 SwapMagazine {
  binding_count=2
  length_seconds=0.5
  object=&self:SwapMagazine
  sequence {bindings [
    Image_Magazine2 {
     id=#2
     possessed_class=Image
     tracks [
      {
       type=MovieSceneFloatTrack
       display_name=Render Opacity
       sections [{type=MovieSceneFloatSection, end_frame=4, start_frame=0, channels [None {type=MovieSceneScriptingFloatChannel, keys [[0,0.5],[4.9623,1]]}]}]
      }
      {
       type=MovieScene2DTransformTrack
       sections [{
         type=MovieScene2DTransformSection
         end_frame=4
         start_frame=0
         channels [
          Translation.X {type=MovieSceneScriptingFloatChannel, keys [[0,52.9604],[4.9623,69.5923]]}
          Translation.Y {type=MovieSceneScriptingFloatChannel, keys [[0,70.6081],[4.9623,57.7624]]}
          Angle {type=MovieSceneScriptingFloatChannel, keys [[0,-162.2985],[4.9623,-181.6335]]}
          Scale.X {type=MovieSceneScriptingFloatChannel, const [0.9,2]}
          Scale.Y {type=MovieSceneScriptingFloatChannel, const [0.9,2]}
          Shear.X {type=MovieSceneScriptingFloatChannel, keys [[0,-2.6139],[4.9623,0]]}
          Shear.Y {type=MovieSceneScriptingFloatChannel, keys [[0,-9.3894],[4.9623,0]]}
         ]
        }]
      }
     ]
    }
    Image_Magazine1 {
     id=#3
     possessed_class=Image
     tracks [
      {
       type=MovieSceneFloatTrack
       display_name=Render Opacity
       sections [{type=MovieSceneFloatSection, end_frame=4, start_frame=0, channels [None {type=MovieSceneScriptingFloatChannel, keys [[0,1],[4.9623,0.4]]}]}]
      }
      {
       type=MovieScene2DTransformTrack
       sections [{
         type=MovieScene2DTransformSection
         end_frame=4
         start_frame=0
         channels [
          Translation.X {type=MovieSceneScriptingFloatChannel, keys [[0,60],[4.9623,51.8077]]}
          Translation.Y {type=MovieSceneScriptingFloatChannel, keys [[0,56],[4.9623,58.0238]]}
          Angle {type=MovieSceneScriptingFloatChannel, keys [[0,180],[4.9623,193.4964]]}
          Scale.X {type=MovieSceneScriptingFloatChannel, keys [[0,0.9],[4.9623,0.8]]}
          Scale.Y {type=MovieSceneScriptingFloatChannel, keys [[0,0.9],[4.9623,0.8]]}
          Shear.X {type=MovieSceneScriptingFloatChannel, keys [[0,0],[4.9623,-0.6781]]}
          Shear.Y {type=MovieSceneScriptingFloatChannel, keys [[0,0],[4.9623,2.7556]]}
         ]
        }]
      }
     ]
    }
   ]}
 }
 SwapMagazine_0 {
  binding_count=2
  length_seconds=0.5
  object=&self:SwapMagazine_0
  sequence {bindings [
    Image_Magazine2 {
     id=#2
     possessed_class=Image
     tracks [{
       type=MovieSceneFloatTrack
       display_name=Render Opacity
       sections [{type=MovieSceneFloatSection, end_frame=10, start_frame=0, channels [None {type=MovieSceneScriptingFloatChannel, keys [[0,0.6],[10,1]]}]}]
      }]
    }
    Image_Magazine1 {
     id=#3
     possessed_class=Image
     tracks [{
       type=MovieSceneFloatTrack
       display_name=Render Opacity
       sections [{type=MovieSceneFloatSection, end_frame=10, start_frame=0, channels [None {type=MovieSceneScriptingFloatChannel, keys [[0,1],[10,0.6]]}]}]
      }]
    }
   ]}
 }
]
graphs [
 EventGraph {nodes [
   n1 = OnSwappedReloadSpeed!() -> (OutputDelegate, then, IsNextReloadFast:bool=~) {is=title K2Node_CustomEvent}
   > n2 = Branch(Condition=Is Valid Timer Handle!(Handle:Timer Handle Structure=Get EquipWaitForStateToSettle() -> (EquipWaitForStateToSettle:Timer Handle Structure)) -> (ReturnValue:bool))
   UpdateActiveMagazine!(
    execute=@n2.else
    self
    NewIndex:int=Select(Option 0:int=1, Option 1:int=0, Index=@n1.IsNextReloadFast) -> (ReturnValue:int)
    PlayAnimation:bool=true
   ) -> (then)
   n3 = OnReloadTimerFinished() {is=title K2Node_CustomEvent}
   Clear and Invalidate Timer by Handle(execute=[@n3.then, @n17.then], Handle=Get ReloadTimer() -> (ReloadTimer:Timer Handle Structure))
   > QueueFadeOut()
   n4 = OnWaitForSettleFinished() {is=title K2Node_CustomEvent}
   n5 = Get DualMagComponent() -> (DualMagComponent:Fort Weapon Component Dual Mag Object Reference)
   > UpdateActiveMagazine(
    NewIndex=Select(Option 0=1, Option 1=0, Index=Get IsNextReloadFast!(self:Fort Weapon Component Dual Mag Object Reference=@n5) -> (IsNextReloadFast:bool))
    PlayAnimation=false
   )
   > SetVisibility(InVisibility=SelfHitTestInvisible)
   > Clear and Invalidate Timer by Handle(Handle=Get EquipWaitForStateToSettle())
   > QueueFadeOut()
   QueueFadeOut() {is=title K2Node_CustomEvent}
   > n6 = Branch(Condition=ShouldWidgetBeHidden!(self:Fort Weapon Reticle Extension Widget Base Object Reference) -> (ReturnValue:bool))
   n7 = Branch(execute=@n6.else, Condition=Is Valid Timer Handle(Handle=Get ReloadTimer()))
   n8 = FadeIn(execute=@n7.else)
   n9 = PlayFadeOut() {is=title K2Node_CustomEvent}
   n10 = Set Timer by Event(execute=@n8.then, Delegate=@n9, Time=Get TimeToOutro() -> (TimeToOutro:double), bLooping=false)
   > Set WaitToFade(WaitToFade:Timer Handle Structure=@n10) -> (Output_Get:Timer Handle Structure)
   n11 = Event OnWeaponReloadStarted!() -> (OutputDelegate, then, ReloadTime:float, ReloadType:EFortWeaponReloadType Enum=ReloadWholeClip) {is=title K2Node_Event}
   > n12 = Set Timer by Event(Delegate=@n3, Time=@n11.ReloadTime, bLooping=false)
   > Set ReloadTimer(ReloadTimer:Timer Handle Structure=@n12) -> (Output_Get:Timer Handle Structure)
   > FadeIn()
   Event Construct()
   > n13 = Set EquippedWeapon(EquippedWeapon:Fort Weapon Object Reference=GetEquippedWeapon!(self:Fort Weapon Reticle Extension Widget Base Object Reference) -> (ReturnValue:Fort Weapon Object Reference)) -> (Output_Get:Fort Weapon Object Reference)
   > Is Valid(InputObject=@n13)
   > n14 = Set DualMagComponent(DualMagComponent:Fort Weapon Component Dual Mag Object Reference=GetComponentByClass!(
     self:Actor Object Reference=@n13
     ComponentClass:Actor Component Class Reference=/Game/Athena/Items/Weapons/Components/FortWeaponComponent_DualMag.FortWeaponComponent_DualMag_C
    ) -> (ReturnValue:Fort Weapon Component Dual Mag Object Reference)) -> (Output_Get:Fort Weapon Component Dual Mag Object Reference)
   > Is Valid(InputObject=@n14)
   > Bind Event to On Local Ammo Changed(self:Fort Weapon Object Reference=Get EquippedWeapon() -> (EquippedWeapon:Fort Weapon Object Reference), Delegate=Create Event()) {is=title K2Node_AddDelegate}
   > Bind Event to Swapped Reload Speed(self:Fort Weapon Component Dual Mag Object Reference=Get DualMagComponent(), Delegate=Create Event()) {is=title K2Node_AddDelegate}
   > n15 = Branch(Condition=ShouldWidgetBeHidden())
   SetVisibility(execute=@n15.else, InVisibility=Hidden)
   > Set CurrentMagIndex(CurrentMagIndex:int=-1) -> (Output_Get:int=0)
   > n16 = Set Timer by Event(Delegate=@n4, Time=0.4, bLooping=false)
   > Set EquipWaitForStateToSettle(EquipWaitForStateToSettle:Timer Handle Structure=@n16) -> (Output_Get:Timer Handle Structure)
   n17 = Event OnWeaponReloadCancelled() {is=title K2Node_Event}
   Event Destruct()
   > Clear and Invalidate Timer by Handle(Handle=Get ReloadTimer())
   > Clear and Invalidate Timer by Handle(Handle=Get EquipWaitForStateToSettle())
   > Clear and Invalidate Timer by Handle(Handle=Get WaitToFade() -> (WaitToFade:Timer Handle Structure))
   n18 = Get EquippedWeapon()
   > Is Valid(InputObject=@n18)
   > Unbind all Events from On Local Ammo Changed!(execute:Exec, self:Fort Weapon Object Reference=@n18) -> (then:Exec) {is=title K2Node_ClearDelegate}
   n19 = Get DualMagComponent()
   > Is Valid(InputObject=@n19)
   > Unbind all Events from Swapped Reload Speed!(execute:Exec, self:Fort Weapon Component Dual Mag Object Reference=@n19) -> (then:Exec) {is=title K2Node_ClearDelegate}
   n20 = Event PreConstruct()
   > n21 = Branch(Condition=@n20.IsDesignTime)
   n22 = Get AmmoBarMaterial() -> (AmmoBarMaterial:Material Instance Object Reference)
   SetBrushFromMaterial(execute=[@n23.then_0, @n21.then], self=Get Image_Magazine1() -> (Image_Magazine1:Image Object Reference), Material=@n22)
   > SetBrushFromMaterial(self=Get Image_Magazine2() -> (Image_Magazine2:Image Object Reference), Material=@n22)
   Event OnInitialized()
   > n23 = Sequence()
   SetupWeapon(execute=@n23.then_1)
   n24 = Event OnTargetingChanged!() -> (OutputDelegate, then, bIsADS:bool) {is=title K2Node_Event}
   > n25 = Sequence()
   > Branch(Condition=@n24.bIsADS)
   > QueueFadeOut() {is=title K2Node_CallFunction}
   n26 = For Each Loop(
    Exec=@n25.then_1
    Array:Array of Material Instance Dynamic Object References=Get MagazineDynamicMaterials() -> (MagazineDynamicMaterials:Array of Material Instance Dynamic Object References)
   ) -> (Array Element:Material Instance Dynamic Object Reference)
   > SetScalarParameterValue(self=@n26, ParameterName=Get IsInADSParam() -> (IsInADSParam:Name=None), Value=SelectFloat(A=1, B=0, bPickA=@n24.bIsADS))
   Event OnAmmoClipAttributeChanged!() -> (OutputDelegate, then, AmmoClipValue:float) {is=title K2Node_Event}
   > UpdateAmmoSegments()
   PlayAnimation(execute=@n9.then, InAnimation=Get FadeOut() -> (FadeOut:Widget Animation Object Reference))
  ]}
 FadeIn {nodes [
   FadeIn() {is=title K2Node_FunctionEntry}
   > SetRenderOpacity(self=Get Overlay_Magazines() -> (Overlay_Magazines:Overlay Object Reference), InOpacity=1)
   > Clear and Invalidate Timer by Handle(Handle=Get WaitToFade())
   > StopAnimation(InAnimation=Get FadeOut())
  ]}
 OnLocalAmmoChanged {nodes [
   n1 = OnLocalAmmoChanged!() -> (then, LocalCount:int, LocalRemaining:int) {is=title K2Node_FunctionEntry}
   > QueueFadeOut()
   > n2 = Branch(Condition=Is Valid Timer Handle(Handle=Get ReloadTimer()))
   ResetAmmo!(execute=@n2.else, self, LocalCount:int=@n1, LocalRemaining:int=@n1.LocalRemaining) -> (then)
  ]}
 ResetAmmo {nodes [
   n1 = ResetAmmo!() -> (then, LocalCount:int, LocalRemaining:int) {is=title K2Node_FunctionEntry}
   > Set TotalAmmo(TotalAmmo:int=@n1.LocalRemaining) -> (Output_Get:int=0)
   > n2 = For Each Loop(Array=Get MagazineDynamicMaterials())
   > Set CurrentMaterial(CurrentMaterial:Material Instance Dynamic Object Reference=@n2) -> (Output_Get:Material Instance Dynamic Object Reference)
   > n3 = Branch(Condition="integer < integer"(A=@n2.Array Index, B=Get CurrentMagIndex() -> (CurrentMagIndex:int=0)))
   > SetScalarParameterValue(self=Get CurrentMaterial() -> (CurrentMaterial:Material Instance Dynamic Object Reference), ParameterName=Progress, Value=0)
   n4 = Branch(execute=@n3.else, Condition="Equal (Integer)"(A=@n2.Array Index, B=Get CurrentMagIndex() -> (CurrentMagIndex=0)))
   n5 = GetBulletsPerClip!(self:Fort Weapon Object Reference=Get EquippedWeapon()) -> (ReturnValue:int=0)
   n6 = Get LocalCount() -> (LocalCount:int=0)
   > SetScalarParameterValue(
    self=Get CurrentMaterial()
    ParameterName=Progress
    Value=float / float(A="To Float (Integer)"(InInt=@n6), B="To Float (Integer)"(InInt=@n5))
   )
   > Set TotalAmmo(TotalAmmo=int - int(A=Get TotalAmmo() -> (TotalAmmo:int=0), B=Get LocalCount() -> (LocalCount=0))) -> (Output_Get=0)
   n7 = "To Float (Integer)"(InInt=GetBulletsPerClip(self=Get EquippedWeapon()) -> (ReturnValue=0))
   n8 = "To Float (Integer)"(InInt=Get TotalAmmo() -> (TotalAmmo=0))
   SetScalarParameterValue(execute=@n4.else, self=Get CurrentMaterial(), ParameterName=Progress, Value="Max (Float)"(A=float / float(A=@n8, B=@n7), B=0))
   n9 = Get EquippedWeapon()
   > Set TotalAmmo(TotalAmmo=int - int(A=Get TotalAmmo() -> (TotalAmmo=0), B=GetBulletsPerClip(self=@n9) -> (ReturnValue=0))) -> (Output_Get=0)
  ]}
 SetupWeapon {nodes [
   SetupWeapon() {is=title K2Node_FunctionEntry}
   > n1 = GetDynamicMaterial(self=Get Image_Magazine1())
   > Add(
    TargetArray:Array of Material Instance Dynamic Object References=Get MagazineDynamicMaterials()
    NewItem:"Material Instance Dynamic Object Reference (by ref)"=@n1
   ) {is=title K2Node_CallArrayFunction}
   > n2 = GetDynamicMaterial(self=Get Image_Magazine2())
   > Add(TargetArray=Get MagazineDynamicMaterials(), NewItem=@n2)
   > UpdateAmmoSegments()
   > n3 = For Each Loop(Array=Get MagazineDynamicMaterials())
   > SetScalarParameterValue(self=@n3, ParameterName=Get IsInADSParam() -> (IsInADSParam=None), Value=0)
  ]}
 UpdateActiveMagazine {nodes [
   n1 = UpdateActiveMagazine!() -> (then, NewIndex:int, PlayAnimation:bool=~) {is=title K2Node_FunctionEntry}
   > n2 = Set CorrectedNewIndex(CorrectedNewIndex:int=@n1) -> (Output_Get:int=0)
   n3 = Get MagazineDynamicMaterials()
   > n4 = Branch(Condition="integer >= integer"(A=@n2, B=Length(TargetArray:Array of Material Instance Dynamic Object References=@n3)))
   > n5 = Set CorrectedNewIndex(CorrectedNewIndex=0) -> (Output_Get=0)
   Branch(
    execute=[@n5.then, @n11.then, @n9.else]
    Condition="Not Equal (Integer)"(A=Get CurrentMagIndex() -> (CurrentMagIndex=0), B=Get CorrectedNewIndex() -> (CorrectedNewIndex:int=0))
   )
   n6 = Get CorrectedNewIndex() -> (CorrectedNewIndex=0)
   n7 = Get SwapMagazine() -> (SwapMagazine:Widget Animation Object Reference)
   > PlayAnimation(
    InAnimation=@n7
    StartAtTime=Select(Option 0:float=GetEndTime(self=@n7), Option 1:float=0, Index=Get PlayAnimation() -> (PlayAnimation:bool)) -> (ReturnValue:float)
    PlayMode=Select(Option 0:EUMGSequencePlayMode Enum, Option 1:EUMGSequencePlayMode Enum=Reverse, Index="Equal (Integer)"(A=@n6)) -> (ReturnValue:EUMGSequencePlayMode Enum)
   )
   > Set CurrentMagIndex(CurrentMagIndex=Get CorrectedNewIndex() -> (CorrectedNewIndex=0)) -> (Output_Get=0)
   n8 = Get EquippedWeapon()
   > ResetAmmo(
    LocalCount=GetMagazineAmmoCount!(self:Fort Weapon Object Reference=@n8) -> (ReturnValue:int=0)
    LocalRemaining=GetRemainingAmmo!(self:Fort Weapon Object Reference=@n8) -> (ReturnValue:int=0)
   ) {is=title K2Node_CallFunction}
   n9 = Branch(execute=@n4.else, Condition="integer < integer"(A=Get CorrectedNewIndex() -> (CorrectedNewIndex=0)))
   n10 = Get MagazineDynamicMaterials()
   > n11 = Set CorrectedNewIndex(CorrectedNewIndex=int - int(A=Length(TargetArray=@n10))) -> (Output_Get=0)
  ]}
 UpdateAmmoSegments {nodes [
   UpdateAmmoSegments() {is=title K2Node_FunctionEntry}
   n1 = GetBulletsPerClip(self=GetEquippedWeapon()) -> (ReturnValue=0)
   n2 = int * int(A:int=@n1, B:int=4) -> (ReturnValue:int=0) {is=title K2Node_CommutativeAssociativeBinaryOperator}
   > Set Segments(Segments:double=Select(Option 0:double="To Float (Integer)"(InInt=@n2), Option 1:double=1, Index="Equal (Integer)"(A=@n1, B=1) {is=title K2Node_CallFunction}) -> (ReturnValue:double)) -> (Output_Get:double)
   > n3 = For Each Loop(Array=Get MagazineDynamicMaterials())
   > SetScalarParameterValue(self=@n3, ParameterName=Segments, Value=Get Segments() -> (Segments:double))
  ]}
]
variables [
 AmmoBarMaterial {type=MaterialInstance, default=/Game/Athena/HUD/Materials/M_UI_RadialProgress_Ammo}
 CurrentMagIndex {type=integer, default=-1}
 DualMagComponent {type=FortWeaponComponent_DualMag_C, default=null}
 EquippedWeapon {type=FortWeapon, default=null}
 EquipWaitForStateToSettle {type=TimerHandle}
 IsInADSParam {type=string, default=IsInADS}
 MagazineDynamicMaterials {type="array<MaterialInstanceDynamic>", default []}
 ReloadTimer {type=TimerHandle}
 SavedMagState {type=integer}
 TimeToOutro {type=number, default=1.5}
 WaitToFade {type=TimerHandle}
]
widgets {
 info {
  inheritedWidgetCount=1
  parentClass=/Script/FortniteUI.FortWeaponReticleExtensionWidgetBase
  rootWidgetClass=&Overlay
  widgetCount=4
 }
 widgets [
  Overlay_Magazines {widgetClassPath=&Overlay, bIsVariable=true, widget=Overlay_Magazines}
  Image_Magazine1 {
   widgetClassPath=&Image
   bIsVariable=true
   parent=Overlay_Magazines
   widget=Image_Magazine1
   properties {
    brush {
     bIsDynamicallyLoaded=false
     drawAs=Image
     imageType=NoImage
     mirroring=NoMirror
     resourceName=None
     resourceObject=&M_UI_RadialProgress_Ammo
     tiling=NoTile
     imageSize=<80 80>
     margin=<ltrb 0 0 0 0>
     outlineSettings {
      bUseBrushTransparency=false
      roundingType=HalfHeightRadius
      width=0
      color {colorUseRule=UseColor_Specified, specifiedColor=<rgba 0 0 0 0>}
      cornerRadii=<0 0 0 0>
     }
     tintColor {colorUseRule=UseColor_Specified, specifiedColor=<rgba 1 1 1 1>}
     uVRegion {bIsValid=false, max=<0 0>, min=<0 0>}
    }
    renderTransform {angle=180, scale=<0.9 0.9>, shear=<0 0>, translation=<60 56>}
   }
   slot {class=OverlaySlot, properties {horizontalAlignment=HAlign_Center, verticalAlignment=VAlign_Center}}
  }
  Image_Magazine2 {
   widgetClassPath=&Image
   bIsVariable=true
   parent=Overlay_Magazines
   widget=Image_Magazine2
   properties {
    renderOpacity=0.2
    brush {
     bIsDynamicallyLoaded=false
     drawAs=Image
     imageType=NoImage
     mirroring=NoMirror
     resourceName=None
     resourceObject=&M_UI_RadialProgress_Ammo
     tiling=NoTile
     imageSize=<80 80>
     margin=<ltrb 0 0 0 0>
     outlineSettings {
      bUseBrushTransparency=false
      roundingType=HalfHeightRadius
      width=0
      color {colorUseRule=UseColor_Specified, specifiedColor=<rgba 0 0 0 0>}
      cornerRadii=<0 0 0 0>
     }
     tintColor {colorUseRule=UseColor_Specified, specifiedColor=<rgba 1 1 1 1>}
     uVRegion {bIsValid=false, max=<0 0>, min=<0 0>}
    }
    renderTransform {angle=-164.941, scale=<0.9 0.9>, shear=<-2 2>, translation=<52 70>}
   }
   slot {class=OverlaySlot, properties {horizontalAlignment=HAlign_Center, verticalAlignment=VAlign_Center}}
  }
  TutorialHighlight {widgetClassPath=&TutorialHighlight}
 ]
}
