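# WBP_FlipMagARProto_Reticle is a reticle widget for the Flipmag AR prototype weapon in HopscotchWeaponsGameplay.
# It is a child of the Athena reticle extension widget WBP_DualMag_Reticle.
# It adds no widgets, variables or event graph logic of its own, so any layout and behavior come from that dual-mag parent.
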WidgetBlueprint /HopscotchWeaponsGameplay/Items/FlipmagAR/UI/WBP_FlipMagARProto_Reticle
graphs [EventGraph {links [], nodes []}]
variables []
widgets {
 info {
  inheritedWidgetCount=0
  parentClass=/Game/Athena/HUD/reticles/ReticleExtensionWidgets/WBP_DualMag_Reticle.WBP_DualMag_Reticle_C
  rootWidgetClass=null
  widgetCount=0
 }
 widgets []
}
